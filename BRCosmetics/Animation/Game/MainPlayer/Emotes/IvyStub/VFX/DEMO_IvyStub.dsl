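; DEMO_IvyStub is a Blueprint, apparently a demo actor for previewing the IvyStub emote VFX.
; On BeginPlay it calls PlayAnimation on its M_MED_Base component with looping enabled, though the call names no animation asset.
; Tick, actor overlap and the construction script are empty.
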(event EventBeginPlay
  (Components|Animation|PlayAnimation (Variables|Default|GetM_MED_Base) 0 true))

(event Collision|EventActorBeginOverlap
  (params (OtherActor :type Actor%20Object%20Reference)))

(event EventTick
  (params (DeltaSeconds :type Float%20%28single-precision%29)))

(fn ConstructionScript)
