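# Niagara system for the IvyStub emote that puts a fading glow on the player character's skeletal mesh.
# Glow_Body, Glow_Feet and AuraGlow_Mobile burst-spawn long-lived particles that follow bones or sockets, picked by partition index or, for the feet, a fixed index.
# Body and feet use mesh renderers; the mobile aura adds sprites sized per partition.
# Alpha curves fade each layer in and out, with intensity switched by front-end preview and, for the aura, time of day.
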NiagaraSystem /BRCosmetics/Animation/Game/MainPlayer/Emotes/IvyStub/VFX/NS_Emote_IvyStub_BodyGlow
dependencies {
 usedDataInterfaces [&NiagaraDataInterfaceCurve, &NiagaraDataInterfaceArrayFloat2, &NiagaraDataInterfaceArrayFloat]
 usedDynamicInputs [
  &SelectFloatFromArray
  &Lerp_Float
  &Waveform
  &RandomRangeFloat
  &SetBoolByFloatComparison
  &RemapRange
  &Multiply_Float
  &MakeFloatFromLinearColor
  &InvertBool
 ]
 usedModules [
  &SystemState
  &self:SystemSpawnScript.SystemScriptSource.SystemScriptGraph.NiagaraNodeAssignment_5.SetVariables_#1
  &InterpolateOverTime
  &self:SystemSpawnScript.SystemScriptSource.SystemScriptGraph.NiagaraNodeAssignment_0.SetVariables_#2
  &self:SystemSpawnScript.SystemScriptSource.SystemScriptGraph.NiagaraNodeAssignment_1.SetVariables_#3
  &self:SystemSpawnScript.SystemScriptSource.SystemScriptGraph.NiagaraNodeAssignment_2.SetVariables_#4
  &self:SystemSpawnScript.SystemScriptSource.SystemScriptGraph.NiagaraNodeAssignment_3.SetVariables_#5
  &self:SystemSpawnScript.SystemScriptSource.SystemScriptGraph.NiagaraNodeAssignment_4.SetVariables_#6
 ]
}
emitters {
 AuraGlow_Mobile {
  emitterUpdateScript [
   /SetVariables(Emitter.GlowOutlineThickness:NiagaraFloat=&FloatFromCurve(
     FloatCurve=DataInterface() {propertyValues {Curve {keys [<at 0 0 RCIM_Cubic>, <at 0.1 1.5>, {arriveTangent=-1.6667, interpMode=RCIM_Cubic, leaveTangent=-1.6667, time=0.8, value=1.5}, <at 1 0>]}}}
     CurveIndex=$System.NormalizedLoopAge
     Scale Curve=0.2
    )) {enabled=false, isSetParameters=true, module=SetVariables_#7}
   &EmitterState(Life Cycle Mode=::NewEnumerator1, Loop Behavior=::NewEnumerator1, Loop Duration=0.1, UseLoopDelay=false)
   &SpawnBurst_Instantaneous(Spawn Count=11, Use Spawn Probability=false, UseLoopCountLimit=false)
  ]
  particleSpawnScript [
   &InitializeParticle(
    Color=<rgba 0.1231 0.5312 0.2682 1>
    Lifetime=7.33
    Mesh Scale Mode=::NewEnumerator4
    Position Offset=<0 0 100>
    Sprite Size Mode=::NewEnumerator3
    Uniform Sprite Size=100
    UsePositionOffset=true
    Write Material Random=true
   )
   &PartitionParticles(Number Of Partitions=11)
  ]
  particleUpdateScript [
   &ParticleState()
   &ScaleSpriteSize(
    Scale Factor=&SelectVector2DFromArray(
     Direct Array Index=$Particles.PartitionParticles.Partition
     Vector 2D Selection Array=DataInterface() {propertyValues {FloatData [<1.25 1.25>, <1.25 1.25>, <2 2>, <1.5 2>, <1.5 1.5>, <1 1>, <1 1>, <0.75 0.75>, <0.75 0.75>, <1.5 2>, <1.5 1.5>]}}
    )
    Scale Sprite Size Mode=::NewEnumerator2
   )
   &ScaleColor(
    Scale Alpha=&FloatFromCurve(
     FloatCurve=DataInterface() {propertyValues {Curve {keys [<at 0 0 RCIM_Cubic>, <at 0.1943 1>, {arriveTangent=-1.2411, interpMode=RCIM_Cubic, leaveTangent=-1.2411, time=0.8, value=1}, <at 1 0>]}}}
     Scale Curve=&MakeCustomFloatFromBool(
      Bool=$User.bIsFrontEndPreview
      True Float=0.18
      False Float=&FN_Select_Float_from_Time_Of_Day(Morning=0.2, Day=0.2, Evening=0.2, Night=0.2)
     )
    )
    ScaleA=true
    ScaleRGB=true
   ) {module=ScaleColor001}
   &SkeletalMeshLocation(
    Bone / Socket Index=$Particles.PartitionParticles.Partition
    Bone Sampling Mode=::NewEnumerator4
    MaskSpawnGroup=false
    Skeletal Mesh=$System.Skeletal Mesh
   )
   &CameraOffset(Camera Offset Amount=&FN_FrontEndFloatModifier(Gameplay=-20, Front End=-50))
   &DynamicMaterialParameters(
    Index 0 Param 1=2
    Index 0 Param 3=0.1
    Index 0 Param 4=1
    Module.Index0Float0Write=true
    Module.Index0Float1Write=true
    Module.Index0Float2Write=true
    Module.Index0Float3Write=true
   )
  ]
  rendererClasses [&NiagaraMeshRendererProperties, &NiagaraSpriteRendererProperties]
  renderers [
   {rendererClass=&NiagaraMeshRendererProperties, rendererIndex=0}
   {rendererClass=&NiagaraSpriteRendererProperties, rendererIndex=1}
  ]
 }
 Glow_Body {
  emitterUpdateScript [
   /SetVariables(Emitter.GlowOutlineThickness:NiagaraFloat=&FloatFromCurve(
     FloatCurve=DataInterface() {propertyValues {Curve {keys [<at 0 0 RCIM_Cubic>, <at 0.1 1.5>, {arriveTangent=-1.6667, interpMode=RCIM_Cubic, leaveTangent=-1.6667, time=0.8, value=1.5}, <at 1 0>]}}}
     CurveIndex=$System.NormalizedLoopAge
     Scale Curve=0.2
    )) {isSetParameters=true, module=SetVariables_#7}
   &EmitterState(Life Cycle Mode=::NewEnumerator1, Loop Behavior=::NewEnumerator1, Loop Duration=0.1, UseLoopDelay=false)
   &SpawnBurst_Instantaneous(Spawn Count=5, Use Spawn Probability=false, UseLoopCountLimit=false)
  ]
  particleSpawnScript [
   &InitializeParticle(Lifetime=7.33, Mesh Scale Mode=::NewEnumerator4, Position Offset=<0 0 100>, UsePositionOffset=true, Write Material Random=true)
   &PartitionParticles(Number Of Partitions=5)
  ]
  particleUpdateScript [
   &ParticleState()
   &ScaleColor(
    Scale Alpha=&FloatFromCurve(
     FloatCurve=DataInterface() {propertyValues {Curve {keys [<at 0 0 RCIM_Cubic>, <at 0.1943 1>, {arriveTangent=-1.2411, interpMode=RCIM_Cubic, leaveTangent=-1.2411, time=0.8, value=1}, <at 1 0>]}}}
     Scale Curve=&MakeCustomFloatFromBool(Bool=$User.bIsFrontEndPreview, True Float=40, False Float=20)
    )
    ScaleA=true
    ScaleRGB=true
   ) {module=ScaleColor001}
   &SkeletalMeshLocation(
    Bone / Socket Index=$Particles.PartitionParticles.Partition
    Bone Sampling Mode=::NewEnumerator4
    MaskSpawnGroup=false
    Orientation Sampling=::NewEnumerator1
    Skeletal Mesh=$System.Skeletal Mesh
   )
   &CameraOffset(Camera Offset Amount=30)
  ]
  rendererClasses [&NiagaraMeshRendererProperties]
  renderers [{rendererClass=&NiagaraMeshRendererProperties, rendererIndex=0}]
 }
 Glow_Feet {
  emitterUpdateScript [
   /SetVariables(Emitter.GlowOutlineThickness:NiagaraFloat=&FloatFromCurve(
     FloatCurve=DataInterface() {propertyValues {Curve {keys [<at 0 0 RCIM_Cubic>, <at 0.1 1.5>, {arriveTangent=-1.6667, interpMode=RCIM_Cubic, leaveTangent=-1.6667, time=0.8, value=1.5}, <at 1 0>]}}}
     CurveIndex=$System.NormalizedLoopAge
     Scale Curve=0.4
    )) {isSetParameters=true, module=SetVariables_#7}
   &EmitterState(Life Cycle Mode=::NewEnumerator1, Loop Behavior=::NewEnumerator1, Loop Duration=0.1, UseLoopDelay=false)
   &SpawnBurst_Instantaneous(Use Spawn Probability=false, UseLoopCountLimit=false)
  ]
  particleSpawnScript [
   &InitializeParticle(
    Lifetime=7.33
    Mesh Scale=<0.5 0.5 1>
    Mesh Scale Mode=::NewEnumerator4
    Position Offset=<0 0 100>
    UsePositionOffset=true
    Write Material Random=true
   )
   &PartitionParticles(Number Of Partitions=7) {enabled=false}
  ]
  particleUpdateScript [
   &ParticleState()
   &ScaleColor(
    Scale Alpha=&FloatFromCurve(
     FloatCurve=DataInterface() {propertyValues {Curve {keys [<at 0 0 RCIM_Cubic>, <at 0.1943 1>, {arriveTangent=-1.2411, interpMode=RCIM_Cubic, leaveTangent=-1.2411, time=0.8, value=1}, <at 1 0>]}}}
     Scale Curve=&MakeCustomFloatFromBool(Bool=$User.bIsFrontEndPreview, True Float=200, False Float=150)
    )
    ScaleA=true
    ScaleRGB=true
   ) {module=ScaleColor001}
   &SkeletalMeshLocation(
    Bone / Socket Index=6
    Bone Sampling Mode=::NewEnumerator4
    MaskSpawnGroup=false
    Orientation Sampling=::NewEnumerator1
    Skeletal Mesh=$System.Skeletal Mesh
   )
   &CameraOffset(Camera Offset Amount=30)
  ]
  rendererClasses [&NiagaraMeshRendererProperties]
  renderers [{rendererClass=&NiagaraMeshRendererProperties, rendererIndex=0}]
 }
}
summary {
 systemName=NS_Emote_IvyStub_BodyGlow
 emitters [
  {emitterName=Glow_Body, rendererClasses [&NiagaraMeshRendererProperties]}
  {emitterName=Glow_Feet, rendererClasses [&NiagaraMeshRendererProperties]}
  {
   emitterName=AuraGlow_Mobile
   rendererClasses [&NiagaraMeshRendererProperties, &NiagaraSpriteRendererProperties]
  }
 ]
 userVariables [
  User.bIsFrontEnd {type=NiagaraBool}
  User.bIsFrontEndPreview {type=NiagaraBool}
  User.Local Alpha Preview {type=NiagaraFloat, defaultValue=0.4, description=""}
  User.Other Alpha Preview {type=NiagaraFloat, defaultValue=0.4, description=""}
  User.Skeletal Mesh Root {
   type=NiagaraDataInterfaceSkeletalMesh
   defaultValue {
    struct=NiagaraExt_VariableValue_DataInterface
    value {dataInterface=&self:NiagaraDataInterfaceSkeletalMesh_0}
   }
  }
  User.SquareRingColor {type=LinearColor, description="", defaultValue {struct=LinearColor, value=<rgba 0 0.3611 0.7083 1>}}
 ]
}
