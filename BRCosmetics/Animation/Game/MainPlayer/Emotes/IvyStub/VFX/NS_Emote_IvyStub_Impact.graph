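# Niagara impact effect for the IvyStub player emote: a short burst of smoke and rising air streaks.
# Round and starburst smoke sprites spawn in a shared offset half-cylinder, fading while dynamic material parameters animate.
# Streak sprites spawn in the same shape, launch from an origin below, and stretch as they narrow.
# Active sprites take system time-of-day tint; the mesh Rocks emitter is disabled and Rock_Debris has no renderer.
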NiagaraSystem /BRCosmetics/Animation/Game/MainPlayer/Emotes/IvyStub/VFX/NS_Emote_IvyStub_Impact
dependencies {
 usedDataInterfaces [&NiagaraDataInterfaceCurve, &NiagaraDataInterfaceVector2DCurve, &NiagaraDataInterfaceColorCurve]
 usedDynamicInputs [
  &SimulationPosition
  &MaskVector2dBySpawnGroup
  &Vector2DFromFloat
  &Lerp_Float
  &FN_FrontEndFloatModifier
  &UniformRangedVector
  &ColorFromCurve
  &FNTimeOfDayScalar
 ]
 usedModules [
  &self:SystemSpawnScript.SystemScriptSource.SystemScriptGraph.NiagaraNodeAssignment_0.SetVariables_#1
  &SystemState
  &self:SystemSpawnScript.SystemScriptSource.SystemScriptGraph.NiagaraNodeAssignment_2.SetVariables_#2
 ]
}
emitters {
 AirStreaks_upward {
  emitterUpdateScript [
   &EmitterState(
    Life Cycle Mode=::NewEnumerator1
    Loop Behavior=::NewEnumerator1
    Loop Delay=0.1
    Loop Duration=0.2
    Scalability Mode=::NewEnumerator1
    UseLoopDelay=true
   )
   &SpawnRate(SpawnRate=25, Use Spawn Probability=false)
   &SpawnBurst_Instantaneous(Spawn Count=12, Use Spawn Probability=false, UseLoopCountLimit=false)
  ]
  particleSpawnScript [
   /SetVariables(Particles.RandomFloat=&UniformRangedFloat(Minimum=0, Maximum=3)) {isSetParameters=true}
   &InitializeParticle(
    Color=&Multiply_LinearColorByFloat(Linear Color=$System.TODColor, Float=$System.TODFloat)
    Lifetime Max=0.3
    Lifetime Min=0.15
    Lifetime Mode=::NewEnumerator1
    Sprite Size Max=<25 100>
    Sprite Size Min=<10 50>
    Sprite Size Mode=::NewEnumerator2
    UsePositionOffset=false
    Write Material Random=true
   ) {module=InitializeParticle001}
   /SetVariables(Particles.SpriteSize=&Multiply_Vector2D_ByFloat(Float=&Multiply_Float(A=1, B=1), Vector2D=$Particles.SpriteSize)) {isSetParameters=true}
   &ShapeLocation(
    MaskSpawnGroup=false
    Non Uniform Scale=&Multiply_VectorByFloat(Vector=<1 1 0>, Float=1)
    Override Seed=false
    Shape Primitive=::NewEnumerator4
    UseOwnerScale=true
   ) {enabled=false}
   &AddVelocity(Module.Use Falloff Distance=false, Origin Offset=<0 0 -500>, Override Seed=false, Velocity Mode=::NewEnumerator1)
   &ApplyInitialForces()
   &SubUVAnimation(End Frame=3, SubUV Animation Mode:ENiagara_SubUVLookupMode=::NewEnumerator2)
   &CameraOffset(Camera Offset Amount=20)
   &ShapeLocation(
    Cylinder Height=1
    Cylinder Radius=150
    Hemicircle X=true
    MaskSpawnGroup=false
    Offset=<0 50 0>
    Offset Mode=::NewEnumerator0
    Override Seed=false
    Rotation Angle=0.2
    Rotation Axis=<0 0 1>
    Rotation Mode=::NewEnumerator1
    Shape Primitive=::NewEnumerator1
    Surface Only=false
    Surface Only Band Thickness=1
    UseOwnerScale=true
   ) {module=ShapeLocation001}
  ]
  particleUpdateScript [
   &ParticleState()
   &SolveForcesAndVelocity(Clamp Velocity=false, Limit Acceleration=false)
   &ScaleSpriteSize(Scale Factor=&Vector2DFromCurve(
     Vector2Curve=DataInterface() {propertyValues {XCurve {keys [<at 0.0036 3>,<at 0.9967 0.0582>]}, YCurve {keys [<at 0 1>,<at 1.002 3.7227>]}}}
     ScaleCurve=&Multiply_Vector2D_ByFloat(Float=0.65, Vector2D=<1 1>)
    ))
   &ScaleColor(
    Scale Alpha=&FloatFromCurve(
     FloatCurve=DataInterface() {propertyValues {Curve {keys [
         {interpMode=RCIM_Cubic, leaveTangent=-1.9244, tangentMode=RCTM_Break, time=0, value=1}
         {interpMode=RCIM_Cubic, tangentMode=RCTM_Break, time=1, value=0}
        ]}}}
     Scale Curve=&RandomRangeFloat(Minimum=0.2, Maximum=0.4)
    )
    Scale RGB=&VectorFromFloat(Value=&RandomRangeFloat(Minimum=0.2, Maximum=0.65))
    ScaleA=true
    ScaleRGB=true
   )
   &DynamicMaterialParameters(
    Index 0 Param 2=1
    Index 0 Param 3=1
    Index 0 Param 4=&FloatFromCurve(FloatCurve=DataInterface() {propertyValues {Curve {keys [<at 0.3 1>,<at 1 0.25>]}}})
    Module.Param0WriteEnabled=true
    Module.Param1WriteEnabled=true
    Module.Param2WriteEnabled=true
    Module.Param3WriteEnabled=true
   )
  ]
  rendererClasses [&NiagaraSpriteRendererProperties]
  renderers [{rendererClass=&NiagaraSpriteRendererProperties}]
 }
 Radial_Smoke_Round {
  emitterUpdateScript [
   &EmitterState(Life Cycle Mode=::NewEnumerator1, Loop Behavior=::NewEnumerator1, Loop Duration=0.3, Scalability Mode=::NewEnumerator1, UseLoopDelay=false)
   &SpawnRate(SpawnRate=35, Use Spawn Probability=false)
   &SpawnBurst_Instantaneous(Spawn Count=3, Spawn Group=1, Use Spawn Probability=false, UseLoopCountLimit=false) {module=SpawnBurst_Instantaneous001}
  ]
  particleSpawnScript [
   &InitializeParticle(
    Color=&Multiply_LinearColorByFloat(Linear Color=$System.TODColor, Float=$System.TODFloat)
    Lifetime=&RandomRangeFloat(Minimum=0.45, Maximum=0.65)
    Position Offset=&RandomRangeVector(Minimum=<0 0 0>, Maximum=<0 0 10>)
    Sprite Rotation Mode=::NewEnumerator2
    Sprite Size Mode=::NewEnumerator1
    Uniform Sprite Size Max=200
    Uniform Sprite Size Min=50
    UsePositionOffset=true
    Write Material Random=true
   ) {module=InitializeParticle001}
   &ShapeLocation(
    Cylinder Height=1
    Cylinder Radius=150
    Hemicircle X=true
    MaskSpawnGroup=false
    Offset=<0 50 0>
    Offset Mode=::NewEnumerator0
    Override Seed=false
    Rotation Angle=0.2
    Rotation Axis=<0 0 1>
    Rotation Mode=::NewEnumerator1
    Shape Primitive=::NewEnumerator1
    Surface Only=false
    Surface Only Band Thickness=1
    UseOwnerScale=true
   ) {module=ShapeLocation001}
  ]
  particleUpdateScript [
   &ParticleState()
   &SpriteRotationRate(Rotation Rate=&RandomRangeFloat(Minimum=-50, Maximum=50))
   &ScaleSpriteSize(Uniform Curve Sprite Scale=DataInterface() {propertyValues {Curve {keys [
        {arriveTangent=4.8725, interpMode=RCIM_Cubic, leaveTangent=4.8725, tangentMode=RCTM_User, time=0, value=0}
        <at 0.2 0.8>
        {tangentMode=RCTM_User, time=1, value=1}
       ]}}})
   &DynamicMaterialParameters(
    Index 0 Param 1=&FloatFromCurve(FloatCurve=DataInterface() {propertyValues {Curve {defaultValue=1, keys [{tangentMode=RCTM_User, time=0, value=0}, {tangentMode=RCTM_User, time=1, value=1}]}}})
    Index 0 Param 2=&RandomRangeFloat(Minimum=0, Maximum=2)
    Index 0 Param 3=1
    Index 0 Param 4=1
    Index 1 Param 1=&FloatFromCurve(FloatCurve=DataInterface() {propertyValues {Curve {keys [
         {arriveTangent=-2.3163, interpMode=RCIM_Cubic, time=0, value=2}
         {arriveTangent=-1.9019, interpMode=RCIM_Cubic, leaveTangent=-1.9019, tangentMode=RCTM_User, time=1, value=0.4}
        ]}}})
    Index 1 Param 2=0.4
    Index 1 Param 3=0.6
    Index 1 Param 4=1
    Index 2 Param 1=&FloatFromCurve(FloatCurve=DataInterface() {propertyValues {Curve {keys [<at 0.46 0>,<at 1 1>]}}})
    Index 2 Param 2=45
    Module.Index0Float0Write=true
    Module.Index0Float1Write=true
    Module.Index0Float2Write=true
    Module.Index0Float3Write=true
    Module.Index1Float0Write=true
    Module.Index1Float1Write=true
    Module.Index1Float2Write=true
    Module.Index1Float3Write=true
    Module.Index2Float0Write=true
    Module.Index2Float1Write=true
    Module.Index2Float2Write=true
    Module.Index2Float3Write=true
    Write Parameter Index 1=::NewEnumerator1
    Write Parameter Index 2=::NewEnumerator1
   )
   &ScaleColor(
    Scale Alpha=&FloatFromCurve(
     FloatCurve=DataInterface() {propertyValues {Curve {keys [<at 0.25 1 RCIM_Cubic>,<at 1 0>]}}}
     Scale Curve=&RandomRangeFloat(Minimum=0.35, Maximum=0.5)
    )
    Scale RGB=&VectorFromFloat(Value=&RandomRangeFloat(Minimum=0.2, Maximum=0.5))
    ScaleA=true
    ScaleRGB=true
   )
  ]
  rendererClasses [&NiagaraSpriteRendererProperties]
  renderers [{rendererClass=&NiagaraSpriteRendererProperties}]
 }
 Radial_Smoke_Starburst {
  emitterUpdateScript [
   &EmitterState(Life Cycle Mode=::NewEnumerator1, Loop Behavior=::NewEnumerator1, Loop Duration=0.3, Scalability Mode=::NewEnumerator1, UseLoopDelay=false)
   &SpawnRate(SpawnRate=50, Use Spawn Probability=false)
   &SpawnBurst_Instantaneous(Spawn Count=3, Spawn Group=1, Use Spawn Probability=false, UseLoopCountLimit=false) {module=SpawnBurst_Instantaneous001}
  ]
  particleSpawnScript [
   &InitializeParticle(
    Color=&Multiply_LinearColorByFloat(Linear Color=$System.TODColor, Float=$System.TODFloat)
    Lifetime=&RandomRangeFloat(Minimum=0.25, Maximum=0.6)
    Sprite Rotation Mode=::NewEnumerator2
    Sprite Size Mode=::NewEnumerator1
    Uniform Sprite Size Max=150
    Uniform Sprite Size Min=25
    UsePositionOffset=false
    Write Material Random=true
   ) {module=InitializeParticle001}
   &ShapeLocation(
    Cylinder Height=1
    Cylinder Radius=150
    Hemicircle X=true
    MaskSpawnGroup=false
    Offset=<0 50 0>
    Offset Mode=::NewEnumerator0
    Override Seed=false
    Rotation Angle=0.2
    Rotation Axis=<0 0 1>
    Rotation Mode=::NewEnumerator1
    Shape Primitive=::NewEnumerator1
    Surface Only=false
    Surface Only Band Thickness=1
    UseOwnerScale=true
   ) {module=ShapeLocation001}
  ]
  particleUpdateScript [
   &ParticleState()
   &DynamicMaterialParameters(
    Index 0 Param 1=&FloatFromCurve(FloatCurve=DataInterface() {propertyValues {Curve {defaultValue=1, keys [{tangentMode=RCTM_User, time=0, value=0}, {tangentMode=RCTM_User, time=1, value=1}]}}})
    Index 0 Param 2=&RandomRangeFloat(Minimum=0, Maximum=2)
    Index 0 Param 3=1
    Index 0 Param 4=1
    Index 1 Param 1=&FloatFromCurve(FloatCurve=DataInterface() {propertyValues {Curve {keys [
         {arriveTangent=-2.3163, interpMode=RCIM_Cubic, time=0, value=2}
         {arriveTangent=-1.9019, interpMode=RCIM_Cubic, leaveTangent=-1.9019, tangentMode=RCTM_User, time=1, value=0.4}
        ]}}})
    Index 1 Param 2=0.4
    Index 1 Param 3=0.6
    Index 1 Param 4=1
    Index 2 Param 1=&FloatFromCurve(FloatCurve=DataInterface() {propertyValues {Curve {keys [<at 0.46 0>,<at 1 1>]}}})
    Index 2 Param 2=45
    Module.Index0Float0Write=true
    Module.Index0Float1Write=true
    Module.Index0Float2Write=true
    Module.Index0Float3Write=true
    Module.Index1Float0Write=true
    Module.Index1Float1Write=true
    Module.Index1Float2Write=true
    Module.Index1Float3Write=true
    Module.Index2Float0Write=true
    Module.Index2Float1Write=true
    Module.Index2Float2Write=true
    Module.Index2Float3Write=true
    Write Parameter Index 1=::NewEnumerator1
    Write Parameter Index 2=::NewEnumerator1
   )
   &ScaleColor(
    Scale Alpha=&FloatFromCurve(
     FloatCurve=DataInterface() {propertyValues {Curve {keys [<at 0.25 1 RCIM_Cubic>,<at 1 0>]}}}
     Scale Curve=&RandomRangeFloat(Minimum=0.5, Maximum=1)
    )
    Scale RGB=&VectorFromFloat(Value=&RandomRangeFloat(Minimum=0.2, Maximum=0.5))
    ScaleA=true
    ScaleRGB=true
   )
  ]
  rendererClasses [&NiagaraSpriteRendererProperties]
  renderers [{rendererClass=&NiagaraSpriteRendererProperties}]
 }
 Rock_Debris {rendererClasses [], renderers []}
 Rocks {
  bEnabled=false
  emitterUpdateScript [
   &EmitterState(Life Cycle Mode=::NewEnumerator1, Loop Behavior=::NewEnumerator1, Loop Delay=0.1, Loop Duration=0.23, UseLoopDelay=false)
   &SpawnRate(SpawnRate=50, Use Spawn Probability=false)
   &SpawnBurst_Instantaneous(Spawn Count=15, Use Spawn Probability=false)
  ]
  particleSpawnScript [
   &InitializeParticle(
    Color Maximum=<rgba 0.0975 0.1 0.0975 1>
    Color Minimum=<rgba 0.02 0.0194 0.02 1>
    Color Mode=::NewEnumerator2
    Lifetime Max=1
    Lifetime Min=0.4
    Lifetime Mode=::NewEnumerator1
    Mesh Scale Max=<0.05 0.05 0.05>
    Mesh Scale Min=<0.01 0.01 0.01>
    Mesh Scale Mode=::NewEnumerator2
    UsePositionOffset=false
    Write Material Random=true
   )
   &AddVelocity(Velocity=&RandomRangeVector(Minimum=<0 0 50>, Maximum=<1 1 300>))
   &InitialMeshOrientation(
    Rotation=&RandomRangeVector(Minimum=<0 0 0>, Maximum=<1 1 1>)
    Use Orientation Vector=true
    Use Rotation Vector=true
   )
   &ShapeLocation(
    Cylinder Height=1
    Cylinder Radius=150
    Hemicircle X=true
    MaskSpawnGroup=false
    Offset=<0 50 0>
    Offset Mode=::NewEnumerator0
    Override Seed=false
    Rotation Angle=0.2
    Rotation Axis=<0 0 1>
    Rotation Mode=::NewEnumerator1
    Shape Primitive=::NewEnumerator1
    Surface Only=false
    Surface Only Band Thickness=1
    UseOwnerScale=true
   ) {module=ShapeLocation001}
   &PointForce(Module.Use Falloff=false, Module.UseLinearForce=false)
   &ApplyInitialForces()
   &InitialMeshOrientation(Use Rotation Vector=false) {module=InitialMeshOrientation001}
  ]
  particleUpdateScript [
   &GravityForce(Gravity=&RandomRangeVector(Minimum=<0 0 -450>, Maximum=<1 1 -600>))
   &MeshRotationForce(Rotation=&RandomRangeVector(Minimum=<-10 -10 -10>, Maximum=<10 10 10>))
   &SolveForcesAndVelocity(Clamp Velocity=false, Limit Acceleration=false)
   &ParticleState()
   &UpdateMeshOrientation(Rotation Rate=&RandomRangeFloat(Minimum=-1, Maximum=2), Rotation Vector=<1 1 1>)
   &ScaleMeshSize(Scale Factor=&VectorFromFloat(Value=&FloatFromCurve(FloatCurve=DataInterface() {propertyValues {Curve {keys [<at 0 1>, {interpMode=RCIM_Cubic, tangentMode=RCTM_User, time=0.25, value=1}, <at 1 0>]}}})))
  ]
  rendererClasses [&NiagaraMeshRendererProperties]
  renderers [{rendererClass=&NiagaraMeshRendererProperties, rendererIndex=0}]
 }
}
summary {
 systemName=NS_Emote_IvyStub_Impact
 emitters [
  {emitterName=Radial_Smoke_Starburst, rendererClasses [&NiagaraSpriteRendererProperties]}
  {emitterName=AirStreaks_upward, rendererClasses [&NiagaraSpriteRendererProperties]}
  {bEnabled=false, emitterName=Rocks, rendererClasses [&NiagaraMeshRendererProperties]}
  {emitterName=Radial_Smoke_Round, rendererClasses [&NiagaraSpriteRendererProperties]}
  {emitterName=Rock_Debris, rendererClasses []}
 ]
 userVariables [
  User.bIsFrontEnd {type=NiagaraBool}
  User.bIsFrontEndPreview {type=NiagaraBool}
 ]
}
