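# Niagara effect for the IvyStub player emote that draws short-lived speed-line streaks as thin, translucent sprites.
# One emitter combines a burst with a spawn rate, placing particles on a torus and pulling them with point attraction.
# Curves stretch and shrink each sprite and fade its alpha, with time-of-day scaling brightening the alpha at night.
# The emitter uses distance culling, and no shown module reads the exposed User.Speed parameter.
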NiagaraSystem /BRCosmetics/Animation/Game/MainPlayer/Emotes/IvyStub/VFX/NS_Emote_IvyStub_Speed_Lines
dependencies {
 usedDataInterfaces [&NiagaraDataInterfaceCurve, &NiagaraDataInterfaceVector2DCurve]
 usedDynamicInputs []
 usedModules [&SystemState]
}
emitters {FN_Sprite_SpeedLines {
  emitterUpdateScript [
   &EmitterState(
    Enable Distance Culling=true
    Life Cycle Mode=::NewEnumerator1
    Loop Behavior=::NewEnumerator1
    Loop Duration=0.5
    MaxDistance=2500
    Module.bUseMaxDistance=true
    Module.bUseMinDistance=false
    Scalability Mode=::NewEnumerator1
    UseLoopDelay=false
   )
   &SpawnBurst_Instantaneous(Spawn Count=10, Use Spawn Probability=false, UseLoopCountLimit=false)
   &SpawnRate(SpawnRate=100, Use Spawn Probability=false)
  ]
  particleSpawnScript [
   &InitializeParticle(
    Color=&RandomRangeLinearColor(Minimum=<rgba 0.2 0.2 0.2 0.2>, Maximum=<rgba 1 1 1 0.5>)
    Lifetime Max=0.2
    Lifetime Min=0.1
    Lifetime Mode=::NewEnumerator1
    Position Offset Coordinate Space=::Local
    Sprite Size Max=<3 25>
    Sprite Size Min=<1 8>
    Sprite Size Mode=::NewEnumerator2
    UsePositionOffset=true
    Write Material Random=true
   ) {module=InitializeParticle001}
   &TorusLocation(
    Align Torus With Axis=true
    Enable Torus Knot Radius=true
    Handle Radius=2.5
    Knot Helix Loops=1
    Knot Ring Loops=1
    Large Radius=200
    MaskSpawnGroup=false
    Module.Override Randomness=false
    Normalized Angle Around Torus Axis=0.5
    Surface Position=0.5
    Torus Knot Tube Radius=50
    Torus Mode=::NewEnumerator1
    U Distribution=0.5
    V Distribution=0.5
   ) {module=TorusLocation001}
  ]
  particleUpdateScript [
   &ParticleState()
   &ScaleColor(
    Scale Alpha=&FloatFromCurve(
     FloatCurve=DataInterface() {propertyValues {Curve {keys [
         {interpMode=RCIM_Cubic, tangentMode=RCTM_User, time=0, value=0}
         {interpMode=RCIM_Cubic, tangentMode=RCTM_User, time=0.1, value=1}
         {tangentMode=RCTM_User, time=1, value=0}
        ]}}}
     Scale Curve=&FNTimeOfDayScalar(Night Scale=2, Morning Scale=1, Evening Scale=1)
    )
    ScaleA=true
    ScaleRGB=true
   ) {module=ScaleColor001}
   &PointAttractionForce(Attraction Radius=10000000000, Attraction Strength=300, Kill Within Radius=false, Module.Use Falloff=true, UseAttractorOffset=true)
   &SolveForcesAndVelocity(Clamp Velocity=false, Limit Acceleration=false)
   &ScaleSpriteSize(Scale Factor=&Vector2DFromCurve(
     Vector2Curve=DataInterface() {propertyValues {
       XCurve {keys [<at 0 0.5 RCIM_Cubic>, {arriveTangent=0.5, interpMode=RCIM_Cubic, leaveTangent=0.5, time=0.4918, value=1}, <at 1 1 RCIM_Cubic>]}
       YCurve {keys [
         <at 0 0.0245 RCIM_Cubic>
         {arriveTangent=-5.2648, interpMode=RCIM_Cubic, leaveTangent=-5.2648, tangentMode=RCTM_User, time=0.25, value=3}
         <at 1 0 RCIM_Cubic>
        ]}
      }}
     ScaleCurve=<1 1>
    ))
  ]
  rendererClasses [&NiagaraSpriteRendererProperties]
  renderers [{rendererClass=&NiagaraSpriteRendererProperties}]
 }}
summary {
 systemName=NS_Emote_IvyStub_Speed_Lines
 emitters [{emitterName=FN_Sprite_SpeedLines, rendererClasses [&NiagaraSpriteRendererProperties]}]
 userVariables [User.Speed {type=NiagaraFloat, defaultValue=1, description=""}]
}
